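# Niagara swing effect for the male Space Wanderer pickaxe axe, combining burst accents with short, timed trails.
# Bursts supply a mesh flare, blue sprites and two mesh layers whose SpawnGroup lerps drive separate color and purple/blue scale curves.
# Trails add distance-spawned flipbook fire shifting blue to magenta and a ribbon widening across the emitter loop.
# Crystal shard meshes spawn around User.SkeletalMesh, inherit velocity and vary by SpawnGroup while shrinking and fading.
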NiagaraSystem /BRCosmetics/Weapons/FORT_Melee/Pickaxe_Space_Wanderer_Male/VFX/Niagara/NS_Pickaxe_SpaceWanderer_Male_Axe_Swing
dependencies {
 usedDataInterfaces [&NiagaraDataInterfaceCurve, &NiagaraDataInterfaceVectorCurve, &NiagaraDataInterfaceColorCurve]
 usedDynamicInputs [&FN_FrontEndFloatModifier]
 usedModules [&SystemState]
}
emitters {
 Blue_Decal {
  emitterUpdateScript [
   &EmitterState()
   &SpawnBurst_Instantaneous(Spawn Count=2, Use Spawn Probability=false)
  ]
  particleSpawnScript [
   /Niagara/Modules/Spawn/Initialization/InitializeParticle(
    Color=<rgba 0 2.0243 5 1>
    Lifetime=&UniformRangedFloat(Minimum=0.35, Maximum=0.5)
    Module.Write Color=true
    Module.Write Lifetime=true
    Module.Write Mass=false
    Module.Write Material Random=true
    Module.Write Position=false
    Module.Write Scale=false
    Module.Write SpriteRotation=true
    Module.Write SpriteSize=true
    Sprite Rotation=&UniformRangedFloat(Minimum=0, Maximum=359)
    Sprite Size=<50 50>
   )
   &DynamicMaterialParameters(
    Index 0 Param 1=0.5
    Module.Param0WriteEnabled=true
    Module.Param1WriteEnabled=false
    Module.Param2WriteEnabled=true
    Module.Param3WriteEnabled=false
   )
  ]
  particleUpdateScript [
   &CameraOffset(Camera Offset Amount=20)
   &ParticleState()
   &ScaleColor(
    Scale Alpha=&FloatFromCurve(FloatCurve=DataInterface() {propertyValues {Curve {keys [
         {arriveTangent=-0.0178, interpMode=RCIM_Cubic, leaveTangent=-0.0178, tangentMode=RCTM_User, time=0, value=0}
         <at 0.2 1>
         <at 0.6 1 RCIM_Cubic>
         <at 1 0>
        ]}}})
    ScaleA=true
    ScaleRGB=true
   )
  ]
  rendererClasses [&NiagaraSpriteRendererProperties]
  renderers [{rendererClass=&NiagaraSpriteRendererProperties}]
 }
 Fire_Trail {
  emitterUpdateScript [
   &EmitterState(Life Cycle Mode=::NewEnumerator1, Loop Behavior=::NewEnumerator1, Loop Delay=0.0065, Loop Duration=0.314, UseLoopDelay=true)
   &SpawnPerUnit(
    Module.Use Max Movement Threshold=true
    Module.Use Movement Tolerance=true
    Movement Tolerance=5
    Spawn Per Unit=15
    Use Spawn Probability=false
   )
  ]
  particleSpawnScript [
   &InitializeParticle(
    Color=<rgba 1 1 1 0>
    Lifetime=&UniformRangedFloat(Minimum=0.2, Maximum=0.35)
    Module.Write Color=false
    Module.Write Lifetime=true
    Module.Write Mass=false
    Module.Write Material Random=true
    Module.Write Position=false
    Module.Write Scale=false
    Module.Write SpriteRotation=true
    Module.Write SpriteSize=true
    Position=&SimulationPosition()
    Sprite Rotation=&UniformRangedFloat(Minimum=-180, Maximum=180)
    Sprite Size=&Vector2DFromFloat(Value=&FloatFromCurve(
      FloatCurve=DataInterface() {propertyValues {Curve {keys [<at 0 0 RCIM_Constant>,<at 0.25 1 RCIM_Constant>,<at 1 0>]}}}
      CurveIndex=$Emitter.NormalizedLoopAge
      Scale Curve=60
     ))
   )
   &DynamicMaterialParameters(
    Index 0 Param 2=&UniformRangedFloat(Minimum=0, Maximum=1)
    Index 0 Param 3=1
    Index 0 Param 4=&UniformRangedFloat(Minimum=1, Maximum=2)
    Module.Param0WriteEnabled=false
    Module.Param1WriteEnabled=true
    Module.Param2WriteEnabled=false
    Module.Param3WriteEnabled=true
   ) {module=DynamicMaterialParameters002}
   /Niagara/Modules/Spawn/Location/SphereLocation(
    Hemisphere X=true
    MaskSpawnGroup=false
    Module.Override Randomness=false
    Non Uniform Scale=<-4 0.7 0.7>
    Sphere Radius=10
    Surface Only=false
   )
   &InheritVelocity(Module.Limit Speed=true)
  ]
  particleUpdateScript [
   &ParticleState()
   &SolveForcesAndVelocity(Clamp Velocity=false, Limit Acceleration=false)
   /Niagara/Modules/Update/SubUV/SubUVAnimation(Number Of Frames=64)
   &ScaleSpriteSize(Scale Factor=&Vector2DFromFloat(Value=&FloatFromCurve(FloatCurve=DataInterface() {propertyValues {Curve {keys [
          {arriveTangent=9.9004, interpMode=RCIM_Cubic, leaveTangent=1.3466, tangentMode=RCTM_User, time=0, value=0.5}
          {interpMode=RCIM_Cubic, leaveTangent=1.3, time=1, value=0.99}
         ]}}})))
   &Color(
    Color=&ColorFromCurve(ColorCurve=DataInterface() {propertyValues {
       AlphaCurve {keys [<at 0 0.7 RCIM_Cubic>,<at 1 1 RCIM_Cubic>]}
       BlueCurve {keys [<at 0.2346 5 RCIM_Cubic>,<at 1 3.6888 RCIM_Cubic>]}
       GreenCurve {keys [<at 0.2346 1.1022 RCIM_Cubic>,<at 1 0 RCIM_Cubic>]}
       RedCurve {keys [<at 0.2346 0 RCIM_Cubic>,<at 1 5 RCIM_Cubic>]}
      }})
    Scale Alpha=&FloatFromCurve(
     FloatCurve=DataInterface() {propertyValues {Curve {keys [
         {arriveTangent=8.7245, interpMode=RCIM_Cubic, leaveTangent=8.7245, tangentMode=RCTM_User, time=0, value=0}
         <at 0.1 1 RCIM_Cubic>
         {arriveTangent=-0.5281, interpMode=RCIM_Cubic, leaveTangent=-0.5281, tangentMode=RCTM_User, time=1, value=0}
        ]}}}
     Scale Curve=0.2
    )
   )
   &DynamicMaterialParameters(
    Index 0 Param 1=&FloatFromCurve(FloatCurve=DataInterface() {propertyValues {Curve {keys [{arriveTangent=2.7551, interpMode=RCIM_Cubic, leaveTangent=2.7551, tangentMode=RCTM_User, time=0, value=0.3}, <at 0.7 1>]}}})
    Index 0 Param 2=1
    Index 0 Param 3=1
    Index 0 Param 4=1
    Module.Param0WriteEnabled=true
    Module.Param1WriteEnabled=true
    Module.Param2WriteEnabled=true
    Module.Param3WriteEnabled=false
   ) {module=DynamicMaterialParameters001}
  ]
  rendererClasses [&NiagaraSpriteRendererProperties]
  renderers [{rendererClass=&NiagaraSpriteRendererProperties}]
 }
 Front_Flare_Large {
  emitterUpdateScript [
   &EmitterState()
   &SpawnBurst_Instantaneous(Spawn Count=1, Use Spawn Probability=false)
  ]
  particleSpawnScript [
   &SystemLocation(MaskSpawnGroup=false, Offset=<5 0 1>)
   &InitializeParticle(
    Color=<rgba 1 0 3 1>
    Lifetime=0.3
    Mesh Scale=<0.25 0.23 0.23>
    Module.Write Color=true
    Module.Write Lifetime=true
    Module.Write Mass=false
    Module.Write Material Random=true
    Module.Write Position=false
    Module.Write Scale=true
    Module.Write SpriteRotation=false
    Module.Write SpriteSize=false
    Sprite Size=<5 5>
   )
  ]
  particleUpdateScript [
   &ParticleState(Kill Particles When Lifetime Has Elapsed=false, Let Infinitely Lived Particles Die When Emitter Deactivates=true)
   &SolveForcesAndVelocity(Clamp Velocity=false, Limit Acceleration=false)
   &ScaleColor(
    Color Value To Scale=$Particles.Color
    Scale Alpha=&FloatFromCurve(
     FloatCurve=DataInterface() {propertyValues {Curve {keys [
         <at 0 0.2>
         {arriveTangent=-0.2, interpMode=RCIM_Cubic, leaveTangent=-0.2, time=0.2, value=1}
         {arriveTangent=-3.0495, interpMode=RCIM_Cubic, leaveTangent=-3.0495, tangentMode=RCTM_User, time=1, value=0}
        ]}}}
     Scale Curve=1.5
    )
    ScaleA=true
    ScaleRGB=false
   )
  ]
  rendererClasses [&NiagaraMeshRendererProperties]
  renderers [{rendererClass=&NiagaraMeshRendererProperties, rendererIndex=0}]
 }
 Purple_Main {
  emitterUpdateScript [
   &EmitterState()
   &SpawnBurst_Instantaneous(Spawn Count=4, Spawn Time=0.05, Use Spawn Probability=false)
  ]
  particleSpawnScript [
   /Game/Effects/Niagara/Modules/Groups/FN_SetSpawnGroup_v1/FN_SetSpawnGroup(Emitter Bone Count=2)
   &SystemLocation(MaskSpawnGroup=false, Offset=<-20 0 10>)
   &InitializeParticle(
    Color=&LerpLinearColors(
     StartColor=&LerpLinearColors(StartColor=<rgba 1 0 1 2>, EndColor=<rgba 10 0 1 1>, LerpFactor=0)
     EndColor=&LerpLinearColors(StartColor=<rgba 1 1 1 1.333>, EndColor=<rgba 2.5 2.5 2.5 1>, LerpFactor=0)
     LerpFactor=HlslExpression() {hlslExpression=Particles.SpawnGroup}
    )
    Lifetime=0.314
    Material Random:NiagaraFloat=$Particles.MaterialRandom
    Mesh Scale=<0.4 0.2 0.2>
    Module.Write Color=true
    Module.Write Lifetime=true
    Module.Write Mass=false
    Module.Write Material Random=true
    Module.Write Position=false
    Module.Write Scale=true
    Module.Write SpriteRotation=true
    Module.Write SpriteSize=false
    Position=&SimulationPosition()
    Sprite Rotation=&UniformRangedFloat(Minimum=-0.2, Maximum=0.2)
    Sprite Size=<5 5>
   )
   /Niagara/Modules/Spawn/Orientation/V2/InitialMeshOrientation(Rotation=<0 0.065 0>, Use Orientation Vector=false, Use Rotation Vector=true)
  ]
  particleUpdateScript [
   &ParticleState()
   &ScaleColor(
    Scale Alpha=&FloatFromCurve(FloatCurve=DataInterface() {propertyValues {Curve {keys [
         {arriveTangent=8.6621, interpMode=RCIM_Cubic, leaveTangent=8.6621, tangentMode=RCTM_User, time=0, value=0}
         {arriveTangent=2.0494, leaveTangent=2.0494, time=0.0877, value=2}
         {arriveTangent=-6.7253, interpMode=RCIM_Cubic, leaveTangent=-6.7253, tangentMode=RCTM_User, time=0.5175, value=1.9725}
         <at 0.9921 -0.0275>
        ]}}})
    ScaleA=true
    ScaleRGB=true
   )
   /SetVariables(Particles.PurpleScale:Vector3f=&Lerp_Vector(
     A=&Multiply_Vector(
      A=$Particles.Initial.Scale
      B=&VectorFromCurve(VectorCurve=DataInterface() {propertyValues {
         XCurve {keys [
           {arriveTangent=0.0368, interpMode=RCIM_Cubic, leaveTangent=0.0368, tangentMode=RCTM_User, time=0, value=0.2}
           {arriveTangent=2.9999, interpMode=RCIM_Cubic, leaveTangent=-3.9455, tangentMode=RCTM_Break, time=0.3, value=1.25}
           {arriveTangent=-0.0621, interpMode=RCIM_Cubic, tangentMode=RCTM_Break, time=0.7, value=0}
          ]}
         YCurve {keys [
           {arriveTangent=0.0463, interpMode=RCIM_Cubic, leaveTangent=0.0463, tangentMode=RCTM_User, time=-0.0013, value=0.6712}
           {arriveTangent=-2.3608, interpMode=RCIM_Cubic, tangentMode=RCTM_Break, time=0.6798, value=0}
          ]}
         ZCurve {keys [<at -0.0081 0.686 RCIM_Cubic>, {arriveTangent=-2.5755, interpMode=RCIM_Cubic, tangentMode=RCTM_Break, time=0.6798, value=0}]}
        }})
     )
     B=<0 0 0>
     Alpha=HlslExpression() {hlslExpression=Particles.SpawnGroup}
    )) {isSetParameters=true, module=SetVariables_#1}
   /SetVariables(Particles.BlueScale:Vector3f=&Lerp_Vector(
     A=<0 0 0>
     B=&Multiply_Vector(
      A=$Particles.Initial.Scale
      B=&VectorFromCurve(VectorCurve=DataInterface() {propertyValues {
         XCurve {keys [
           {arriveTangent=0.0368, interpMode=RCIM_Cubic, leaveTangent=0.0368, tangentMode=RCTM_User, time=0, value=0.2}
           {arriveTangent=2.9999, interpMode=RCIM_Cubic, leaveTangent=-3.9455, tangentMode=RCTM_Break, time=0.2, value=1.25}
           {arriveTangent=-0.0621, interpMode=RCIM_Cubic, tangentMode=RCTM_Break, time=0.7, value=0}
          ]}
         YCurve {keys [
           {arriveTangent=0.0463, interpMode=RCIM_Cubic, leaveTangent=0.0463, tangentMode=RCTM_User, time=0, value=0.67}
           {arriveTangent=-2.3608, interpMode=RCIM_Cubic, tangentMode=RCTM_Break, time=0.6798, value=0}
          ]}
         ZCurve {keys [<at 0 0.67 RCIM_Cubic>, {arriveTangent=-2.5755, interpMode=RCIM_Cubic, tangentMode=RCTM_Break, time=0.6798, value=0}]}
        }})
     )
     Alpha=HlslExpression() {hlslExpression="float(Particles.SpawnGroup)"}
    )) {isSetParameters=true, module=SetVariables_#2}
  ]
  rendererClasses [&NiagaraMeshRendererProperties]
  renderers [
   {rendererClass=&NiagaraMeshRendererProperties, rendererIndex=0}
   {rendererClass=&NiagaraMeshRendererProperties, rendererIndex=1}
  ]
 }
 Purple_Main001 {
  emitterUpdateScript [
   &EmitterState()
   &SpawnBurst_Instantaneous(Spawn Count=4, Spawn Time=0.05, Use Spawn Probability=false)
  ]
  particleSpawnScript [
   &FN_SetSpawnGroup(Emitter Bone Count=2)
   &SystemLocation(MaskSpawnGroup=false, Offset=<-20 0 0>)
   &InitializeParticle(
    Color=&LerpLinearColors(
     StartColor=&LerpLinearColors(StartColor=<rgba 1 0 1 2>, EndColor=<rgba 10 0 1 1>, LerpFactor=0)
     EndColor=&LerpLinearColors(StartColor=<rgba 1 1 1 1.333>, EndColor=<rgba 2.5 2.5 2.5 1>, LerpFactor=0)
     LerpFactor=HlslExpression() {hlslExpression=Particles.SpawnGroup}
    )
    Lifetime=0.314
    Material Random:NiagaraFloat=$Particles.MaterialRandom
    Mesh Scale=<0.225 0.16 0.16>
    Module.Write Color=true
    Module.Write Lifetime=true
    Module.Write Mass=false
    Module.Write Material Random=true
    Module.Write Position=false
    Module.Write Scale=true
    Module.Write SpriteRotation=true
    Module.Write SpriteSize=false
    Position=&SimulationPosition()
    Sprite Rotation=&UniformRangedFloat(Minimum=-0.2, Maximum=0.2)
    Sprite Size=<5 5>
   )
   &InitialMeshOrientation(Rotation=<0 0.1 0>, Use Orientation Vector=false, Use Rotation Vector=true)
  ]
  particleUpdateScript [
   &ParticleState()
   &ScaleColor(
    Scale Alpha=&FloatFromCurve(FloatCurve=DataInterface() {propertyValues {Curve {keys [
         {arriveTangent=8.6621, interpMode=RCIM_Cubic, leaveTangent=8.6621, tangentMode=RCTM_User, time=0, value=0}
         {arriveTangent=2.0494, leaveTangent=2.0494, time=0.0877, value=2}
         {arriveTangent=-6.7253, interpMode=RCIM_Cubic, leaveTangent=-6.7253, tangentMode=RCTM_User, time=0.5175, value=1.9725}
         <at 0.9921 -0.0275>
        ]}}})
    ScaleA=true
    ScaleRGB=true
   )
   /SetVariables(Particles.PurpleScale:Vector3f=&Lerp_Vector(
     A=&Multiply_Vector(
      A=$Particles.Initial.Scale
      B=&VectorFromCurve(VectorCurve=DataInterface() {propertyValues {
         XCurve {keys [
           {arriveTangent=0.0368, interpMode=RCIM_Cubic, leaveTangent=0.0368, tangentMode=RCTM_User, time=0, value=0.2}
           {arriveTangent=2.9999, interpMode=RCIM_Cubic, leaveTangent=-3.9455, tangentMode=RCTM_Break, time=0.3, value=1.25}
           {arriveTangent=-0.0621, interpMode=RCIM_Cubic, tangentMode=RCTM_Break, time=0.7, value=0}
          ]}
         YCurve {keys [
           {arriveTangent=0.0463, interpMode=RCIM_Cubic, leaveTangent=0.0463, tangentMode=RCTM_User, time=-0.0013, value=0.6712}
           {arriveTangent=-2.3608, interpMode=RCIM_Cubic, tangentMode=RCTM_Break, time=0.6798, value=0}
          ]}
         ZCurve {keys [<at -0.0081 0.686 RCIM_Cubic>, {arriveTangent=-2.5755, interpMode=RCIM_Cubic, tangentMode=RCTM_Break, time=0.6798, value=0}]}
        }})
     )
     B=<0 0 0>
     Alpha=HlslExpression() {hlslExpression=Particles.SpawnGroup}
    )) {isSetParameters=true, module=SetVariables_#1}
   /SetVariables(Particles.BlueScale:Vector3f=&Lerp_Vector(
     A=<0 0 0>
     B=&Multiply_Vector(
      A=$Particles.Initial.Scale
      B=&VectorFromCurve(VectorCurve=DataInterface() {propertyValues {
         XCurve {keys [
           {arriveTangent=0.0368, interpMode=RCIM_Cubic, leaveTangent=0.0368, tangentMode=RCTM_User, time=0, value=0.2}
           {arriveTangent=2.9999, interpMode=RCIM_Cubic, leaveTangent=-3.9455, tangentMode=RCTM_Break, time=0.2, value=1.25}
           {arriveTangent=-0.0621, interpMode=RCIM_Cubic, tangentMode=RCTM_Break, time=0.7, value=0}
          ]}
         YCurve {keys [
           {arriveTangent=0.0463, interpMode=RCIM_Cubic, leaveTangent=0.0463, tangentMode=RCTM_User, time=0, value=0.67}
           {arriveTangent=-2.3608, interpMode=RCIM_Cubic, tangentMode=RCTM_Break, time=0.6798, value=0}
          ]}
         ZCurve {keys [<at 0 0.67 RCIM_Cubic>, {arriveTangent=-2.5755, interpMode=RCIM_Cubic, tangentMode=RCTM_Break, time=0.6798, value=0}]}
        }})
     )
     Alpha=HlslExpression() {hlslExpression="float(Particles.SpawnGroup)"}
    )) {isSetParameters=true, module=SetVariables_#2}
  ]
  rendererClasses [&NiagaraMeshRendererProperties]
  renderers [
   {rendererClass=&NiagaraMeshRendererProperties, rendererIndex=0}
   {rendererClass=&NiagaraMeshRendererProperties, rendererIndex=1}
  ]
 }
 RibbonTrail {
  emitterUpdateScript [
   &EmitterState(Life Cycle Mode=::NewEnumerator1, Loop Behavior=::NewEnumerator1, Loop Delay=0.05, Loop Duration=0.3, UseLoopDelay=true)
   &SpawnRate(SpawnRate=80, Use Spawn Probability=false)
  ]
  particleSpawnScript [&InitializeRibbon(
    Color=<rgba 0 0 0 0.75>
    Lifetime=0.25
    Module.Write Color=true
    Module.Write Lifetime=true
    Module.Write Mass=true
    Module.Write Material Random=true
    Module.Write Position=false
    Module.Write RibbonTwist=false
    Module.Write RibbonWidth=true
    Ribbon Width=&Lerp_Float(A=5, B=120, Alpha=$Emitter.NormalizedLoopAge)
   )]
  particleUpdateScript [
   &ParticleState()
   &ScaleColor(
    Scale Alpha=&FloatFromCurve(FloatCurve=DataInterface() {propertyValues {Curve {keys [<at 0 1>,<at 1 0>]}}})
    ScaleA=true
    ScaleRGB=true
   )
   &ScaleRibbonWidth(Ribbon Width Scale=&FloatFromCurve(FloatCurve=DataInterface() {propertyValues {Curve {keys [<at 0 1 RCIM_Cubic>, {arriveTangent=-1.8739, interpMode=RCIM_Cubic, leaveTangent=-1.8739, tangentMode=RCTM_User, time=1, value=0}]}}}))
  ]
  rendererClasses [&NiagaraRibbonRendererProperties]
  renderers [{rendererClass=&NiagaraRibbonRendererProperties}]
 }
 crystal_shards {
  emitterUpdateScript [
   &EmitterState(Life Cycle Mode=::NewEnumerator1, Loop Behavior=::NewEnumerator1, Loop Delay=0.075, Loop Duration=0.3, UseLoopDelay=true)
   &SpawnRate(Spawn Probability=0, SpawnRate=400, Use Spawn Probability=false)
  ]
  particleSpawnScript [
   &FN_SetSpawnGroup(Emitter Bone Count=2)
   &InitializeParticle(
    Color=&LerpLinearColors(
     StartColor=&MakeLinearColorFromVectorAndFloat(
      "Vector (RGB)"=&MakeVectorFromLinearColorRGB(Color=<rgba 0 80.9736 100 1>)
      "Float (Alpha)"=&Lerp_Float(
       A=&UniformRangedFloat(Minimum=0.8, Maximum=1.2)
       B=&UniformRangedFloat(Minimum=1.5, Maximum=3)
       Alpha=$Emitter.AccelerationRamp
      )
     )
     EndColor=<rgba 50 0 50 1>
     LerpFactor=HlslExpression() {hlslExpression=Particles.SpawnGroup}
    )
    Lifetime=&Lerp_Float(
     A=&UniformRangedFloat(Minimum=0.2, Maximum=0.5)
     B=&UniformRangedFloat(Minimum=0, Maximum=0.5)
     Alpha=HlslExpression() {hlslExpression=Particles.SpawnGroup}
    )
    Mesh Scale=&UniformRangedVector(
     Minimum=&VectorFromFloat(Value=0.02)
     Maximum=&VectorFromFloat(Value=0.04)
    )
    Module.Write Color=true
    Module.Write Lifetime=true
    Module.Write Mass=false
    Module.Write Material Random=true
    Module.Write Position=false
    Module.Write Scale=true
    Module.Write SpriteRotation=false
    Module.Write SpriteSize=false
    Sprite Size=<5 5>
   )
   &SkeletalMeshLocation(
    Bone / Socket Index=1
    Inherit Velocity=false
    MaskSpawnGroup=false
    Mesh Sampling Type=::NewEnumerator5
    Skeletal Mesh=$User.SkeletalMesh
    Socket Sampling Mode=::NewEnumerator5
   )
   &CylinderLocation(
    Cylinder Height=10
    Cylinder Radius=&Lerp_Float(
     A=&UniformRangedFloat(Minimum=0, Maximum=10)
     B=&UniformRangedFloat(Minimum=10, Maximum=20)
     Alpha=HlslExpression() {hlslExpression=Particles.SpawnGroup}
    )
    MaskSpawnGroup=false
    Module.Enable Lathe Profile=false
    Module.Override Randomness=false
    Module.Surface Only=true
    Orientation Axis:ENiagaraOrientationAxis=::YAxis
    Surface Only Band Thickness:NiagaraFloat=&Lerp_Float(A=5, B=2, Alpha=HlslExpression() {hlslExpression=Particles.SpawnGroup})
    Use Endcaps In Surface Only Mode=false
   )
   &InheritVelocity(Module.Limit Speed=true, Velocity Scale=&UniformRangedVector(Minimum=<0 0 0>, Maximum=<1 1 1>))
   &InitialMeshOrientation(
    Rotation=&UniformRangedVector(Minimum=<0 0 0>, Maximum=<1 1 1>)
    Use Orientation Vector=true
    Use Rotation Vector=true
   )
  ]
  particleUpdateScript [
   &ParticleState()
   /Niagara/Modules/Update/Forces/CurlNoiseForce(Noise Frequency=5, Noise Strength=500, Pan Noise=false, Random Seed=1) {enabled=false}
   &SolveForcesAndVelocity(Clamp Velocity=false, Limit Acceleration=false)
   &ScaleMeshSize(Scale Factor=&Lerp_Vector(
     A=&VectorFromFloat(Value=&FloatFromCurve(FloatCurve=DataInterface() {propertyValues {Curve {keys [
           <at 0 1>
           {arriveTangent=-0.5431, interpMode=RCIM_Cubic, leaveTangent=-0.5431, tangentMode=RCTM_User, time=0.3, value=1}
           {arriveTangent=-5.5854, interpMode=RCIM_Cubic, leaveTangent=-5.5854, tangentMode=RCTM_User, time=1, value=0}
          ]}}}))
     B=&VectorFromFloat(Value=&FloatFromCurve(FloatCurve=DataInterface() {propertyValues {Curve {keys [
           <at 0 1>
           {arriveTangent=-0.5431, interpMode=RCIM_Cubic, leaveTangent=-0.5431, tangentMode=RCTM_User, time=0.6, value=1}
           {arriveTangent=-5.5854, interpMode=RCIM_Cubic, leaveTangent=-5.5854, tangentMode=RCTM_User, time=1, value=0}
          ]}}}))
     Alpha=HlslExpression() {hlslExpression=Particles.SpawnGroup}
    ))
   &ScaleColor(
    Scale Alpha=&FloatFromCurve(FloatCurve=DataInterface() {propertyValues {Curve {keys [<at 0 1>,<at 1 0>]}}})
    Scale RGB=&VectorFromFloat(Value=&FloatFromCurve(FloatCurve=DataInterface() {propertyValues {Curve {keys [<at 0 1>,<at 1 0>]}}}))
    ScaleA=true
    ScaleRGB=true
   )
   /Niagara/Modules/Update/Orientation/V2/UpdateMeshOrientation(Rotation Rate:NiagaraFloat=$Particles.Initial.MaterialRandom, Rotation Vector:Vector3f=<1 0 0>)
  ]
  rendererClasses [&NiagaraMeshRendererProperties]
  renderers [{rendererClass=&NiagaraMeshRendererProperties, rendererIndex=0}]
 }
}
summary {
 systemName=NS_Pickaxe_SpaceWanderer_Male_Axe_Swing
 emitters [
  {emitterName=Front_Flare_Large, rendererClasses [&NiagaraMeshRendererProperties]}
  {emitterName=Blue_Decal, rendererClasses [&NiagaraSpriteRendererProperties]}
  {emitterName=crystal_shards, rendererClasses [&NiagaraMeshRendererProperties]}
  {emitterName=Purple_Main001, rendererClasses [&NiagaraMeshRendererProperties]}
  {emitterName=Fire_Trail, rendererClasses [&NiagaraSpriteRendererProperties]}
  {emitterName=RibbonTrail, rendererClasses [&NiagaraRibbonRendererProperties]}
  {emitterName=Purple_Main, rendererClasses [&NiagaraMeshRendererProperties]}
 ]
 userVariables [
  User.bIsFrontEnd {type=NiagaraBool}
  User.bIsFrontEndPreview {type=NiagaraBool, defaultValue=true}
  User.FEVec {type=Vector3f, description="", defaultValue {struct=Vector3f, value=<0 0 0>}}
  User.RibbonIDOffset {type=NiagaraInt32}
  User.SkeletalMesh {
   type=NiagaraDataInterfaceSkeletalMesh
   defaultValue {
    struct=NiagaraExt_VariableValue_DataInterface
    value {dataInterface=&self:NiagaraDataInterfaceSkeletalMesh_2}
   }
  }
 ]
}
